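# Sound cue for opening the Farmer glider, layering its open sound with a delayed whoosh.
# A mixer combines the Farmer open wave, played after a very short delay, with a louder, pitched-down layer that randomly picks one of three medieval hammer whooshes after a longer delay.
# It uses glider open/close attenuation and concurrency settings, a cosmetic glider-open property sheet and the Fort_Gliders_Land sound class.
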SoundCue /BRCosmetics/Sounds/Gliders/Farmer/Glider_Farmer_Open_Cue
root=Mixer_0
nodes [
 {id=Delay_0, children [WavePlayer_0], properties {delayMax=0.03, delayMin=0.03}}
 {id=Delay_1, children [Random_0], properties {delayMax=1.3, delayMin=1.3}}
 {id=Mixer_0, children [Delay_0,Modulator_0], properties {inputVolume [1,1.4]}}
 {id=Modulator_0, children [Delay_1], properties {pitchMax=0.85, pitchMin=0.85}}
 {id=Random_0, children [WavePlayer_1,WavePlayer_2,WavePlayer_3], properties {weights [1,1,1]}}
 {id=WavePlayer_0, properties {soundWaveAssetPtr=&/Glider_Farmer_Open_01}}
 {id=WavePlayer_1, properties {soundWaveAssetPtr=/Game/Sounds/Fort_Weapons/Medieval/Hammer/MedievalHammer_Whoosh_03}}
 {id=WavePlayer_2, properties {soundWaveAssetPtr=/Game/Sounds/Fort_Weapons/Medieval/Hammer/MedievalHammer_Whoosh_04}}
 {id=WavePlayer_3, properties {soundWaveAssetPtr=/Game/Sounds/Fort_Weapons/Medieval/Hammer/MedievalHammer_Whoosh_05}}
]
properties {
 attenuationSettings=/Game/Sounds/Attenuation/Fort_Player_Glider_OpenClose_Athena_Att
 audioPropertiesSheet=/BRCosmetics/Sounds/Settings/PropertySheets/Glider/APS_Cosmetic_Glider_Open
 maxDistance=10000
 soundClassObject=&Fort_Gliders_Land
 volumeMultiplier=0.4
 concurrencySet [&Athena_GliderOpenClose_Conc]
}
